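# Niagara sprite fireworks effect for the Teddys vs Bears truce proximity-emote quest in the HiddenWar discoverable quests.
# A looping-rate parent emitter launches randomly colored, flickering glow shells upward that emit death events.
# Child glow, spark and two confetti emitters, named as burst children, have no active spawners; their event handlers are not shown.
# Water, pontoon and boost user parameters look like leftovers, with no visible module reading them.
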NiagaraSystem /Game/Athena/Items/DiscoverableQuests/HiddenWar/QuestProximityEmote_Truce/NS_Teddys_vs_Bears_Fireworks
dependencies {
 usedDataInterfaces [&NiagaraDataInterfaceCurve, &NiagaraDataInterfaceColorCurve]
 usedDynamicInputs []
 usedModules [/Niagara/Modules/System/SystemLifeCycle, &SystemState]
}
emitters {
 Confetti_Child {
  emitterUpdateScript [
   &EmitterScalability(
    Module.bApplySpawnCountScale=false
    Module.bApplySpawnCountScaleByDistanceFraction=false
    Module.bUseMaxDistance=false
    Module.bUseMinDistance=false
   ) {enabled=false}
   &EmitterLifeCycle(NextLoopDuration=1) {enabled=false}
   &EmitterState(Life Cycle Mode=::NewEnumerator1, Loop Behavior=::NewEnumerator1, Loop Duration=3, UseLoopDelay=false)
   &SpawnBurst_Instantaneous(Spawn Count=15, Spawn Time=0.05, Use Spawn Probability=false) {enabled=false}
  ]
  particleSpawnScript [
   /Niagara/Modules/Spawn/Initialization/InitializeParticle(
    Color=<rgba 1 1 0 1>
    Lifetime=&UniformRangedFloat(Minimum=0.25, Maximum=2)
    Module.Write Color=true
    Module.Write Lifetime=true
    Module.Write Mass=true
    Module.Write Material Random=true
    Module.Write Position=false
    Module.Write Scale=false
    Module.Write SpriteRotation=true
    Module.Write SpriteSize=true
    Position=&SimulationPosition()
    Sprite Rotation=&UniformRangedFloat(Minimum=-180, Maximum=180)
    Sprite Size=&Vector2DFromFloat(Value=&UniformRangedFloat(Minimum=1, Maximum=1.5))
   )
   &AddVelocity(Velocity=&UniformRangedVector(Minimum=<-450 -450 -450>, Maximum=<450 450 450>))
   &GravityForce(Gravity=<0 0 -500>) {enabled=false, module=GravityForce001}
   &ApplyInitialForces()
  ]
  particleUpdateScript [
   &ParticleState()
   &GravityForce(Gravity=&MakeVector(X=0, Y=0, Z=&UniformRangedFloat(Minimum=-250, Maximum=-350)))
   /Niagara/Modules/Update/Forces/CurlNoiseForce(Noise Frequency=0.02, Noise Strength=300, Pan Noise=false)
   &Drag(Drag=4, Use Linear Drag=true, Use Rotational Drag=false)
   &SolveForcesAndVelocity(Clamp Velocity=false, Limit Acceleration=false)
   &ScaleColor(
    Scale RGB=&MakeVectorFromLinearColorRGB(Color=&ColorFromCurve(ColorCurve=DataInterface() {propertyValues {
        AlphaCurve {keys [<at -0.0023 1 RCIM_Cubic>]}
        BlueCurve {keys [<at -0.0023 1 RCIM_Cubic>,<at 0.2484 1>,<at 0.9977 0 RCIM_Cubic>]}
        GreenCurve {keys [<at -0.0023 0 RCIM_Cubic>,<at 0.2484 41>,<at 0.9977 97 RCIM_Cubic>]}
        RedCurve {keys [<at -0.0023 3 RCIM_Cubic>,<at 0.2484 1>,<at 0.9977 1 RCIM_Cubic>]}
       }}))
    ScaleA=true
    ScaleRGB=true
   )
   &ScaleSpriteSize(Scale Factor=&Vector2DFromFloat(Value=&FloatFromCurve(
      FloatCurve=DataInterface() {propertyValues {Curve {keys [
          {arriveTangent=0.0533, interpMode=RCIM_Cubic, leaveTangent=0.0533, tangentMode=RCTM_User, time=-0.0022, value=1.0215}
          {arriveTangent=-1.037, interpMode=RCIM_Cubic, leaveTangent=-1.037, time=0.7543, value=0.841}
          {arriveTangent=-4.9933, interpMode=RCIM_Cubic, leaveTangent=-4.9933, tangentMode=RCTM_User, time=0.946, value=0.0382}
         ]}}}
      Scale Curve=10
     )))
  ]
  rendererClasses [&NiagaraSpriteRendererProperties]
  renderers [{rendererClass=&NiagaraSpriteRendererProperties}]
 }
 Confetti_Ribbons_Child {
  emitterUpdateScript [
   &EmitterScalability(
    Module.bApplySpawnCountScale=false
    Module.bApplySpawnCountScaleByDistanceFraction=false
    Module.bUseMaxDistance=false
    Module.bUseMinDistance=false
   ) {enabled=false}
   &EmitterLifeCycle(NextLoopDuration=1) {enabled=false}
   &EmitterState(Life Cycle Mode=::NewEnumerator1, Loop Behavior=::NewEnumerator1, Loop Duration=3, UseLoopDelay=false)
   &SpawnBurst_Instantaneous(Spawn Count=15, Spawn Time=0.05, Use Spawn Probability=false) {enabled=false}
  ]
  particleSpawnScript [
   &InitializeParticle(
    Color=<rgba 1 1 0 1>
    Lifetime=&UniformRangedFloat(Minimum=0.25, Maximum=2)
    Module.Write Color=true
    Module.Write Lifetime=true
    Module.Write Mass=true
    Module.Write Material Random=true
    Module.Write Position=false
    Module.Write Scale=false
    Module.Write SpriteRotation=true
    Module.Write SpriteSize=true
    Position=&SimulationPosition()
    Sprite Rotation=&UniformRangedFloat(Minimum=-180, Maximum=180)
    Sprite Size=&Vector2DFromFloat(Value=&UniformRangedFloat(Minimum=1, Maximum=1.5))
   )
   &AddVelocity(Velocity=&UniformRangedVector(Minimum=<-450 -450 -450>, Maximum=<450 450 450>))
   &ApplyInitialForces()
  ]
  particleUpdateScript [
   &ParticleState()
   &GravityForce(Gravity=&MakeVector(X=0, Y=0, Z=&UniformRangedFloat(Minimum=-100, Maximum=-200)))
   &CurlNoiseForce(Noise Frequency=0.02, Noise Strength=300, Pan Noise=false)
   &Drag(Drag=3, Use Linear Drag=true, Use Rotational Drag=false)
   &SolveForcesAndVelocity(Clamp Velocity=false, Limit Acceleration=false)
   &ScaleColor(
    Scale RGB=&MakeVectorFromLinearColorRGB(Color=&ColorFromCurve(ColorCurve=DataInterface() {propertyValues {
        AlphaCurve {keys [<at -0.0023 1 RCIM_Cubic>]}
        BlueCurve {keys [<at -0.0023 1 RCIM_Cubic>,<at 0.2484 1>,<at 0.9977 0 RCIM_Cubic>]}
        GreenCurve {keys [<at -0.0023 0 RCIM_Cubic>,<at 0.2484 20>,<at 0.9977 45 RCIM_Cubic>]}
        RedCurve {keys [<at -0.0023 3 RCIM_Cubic>,<at 0.2484 1>,<at 0.9977 1 RCIM_Cubic>]}
       }}))
    ScaleA=true
    ScaleRGB=true
   )
   &ScaleSpriteSize(Scale Factor=&Vector2DFromFloat(Value=&FloatFromCurve(
      FloatCurve=DataInterface() {propertyValues {Curve {keys [
          {arriveTangent=0.0533, interpMode=RCIM_Cubic, leaveTangent=0.0533, tangentMode=RCTM_User, time=-0.0022, value=1.0215}
          {arriveTangent=-1.037, interpMode=RCIM_Cubic, leaveTangent=-1.037, time=0.7543, value=0.841}
          {arriveTangent=-4.9933, interpMode=RCIM_Cubic, leaveTangent=-4.9933, tangentMode=RCTM_User, time=0.946, value=0.0382}
         ]}}}
      Scale Curve=10
     )))
  ]
  rendererClasses [&NiagaraSpriteRendererProperties]
  renderers [{rendererClass=&NiagaraSpriteRendererProperties}]
 }
 Glow_Child001 {
  emitterUpdateScript [
   &EmitterScalability(
    Module.bApplySpawnCountScale=false
    Module.bApplySpawnCountScaleByDistanceFraction=false
    Module.bUseMaxDistance=false
    Module.bUseMinDistance=false
   ) {enabled=false}
   &EmitterLifeCycle(NextLoopDuration=1) {enabled=false}
   &EmitterState(Life Cycle Mode=::NewEnumerator1, Loop Behavior=::NewEnumerator1, Loop Duration=3, UseLoopDelay=false)
   &SpawnBurst_Instantaneous(Spawn Count=15, Spawn Time=0.05, Use Spawn Probability=false) {enabled=false}
  ]
  particleSpawnScript [&InitializeParticle(
    Color=&ColorFromCurve(
     ColorCurve=DataInterface() {propertyValues {
       AlphaCurve {keys [<at -0.0023 1 RCIM_Cubic>]}
       BlueCurve {keys [<at -0.0023 1 RCIM_Cubic>,<at 0.1951 1>,<at 0.4596 0.6202>,<at 0.7292 0.115>,<at 0.9977 0.099 RCIM_Cubic>]}
       GreenCurve {keys [<at -0.0023 0.0677 RCIM_Cubic>,<at 0.1951 0.3883>,<at 0.4596 1>,<at 0.7292 1>,<at 0.9977 0.427 RCIM_Cubic>]}
       RedCurve {keys [<at -0.0023 0.9381 RCIM_Cubic>,<at 0.1951 0.0434>,<at 0.4596 0>,<at 0.7292 0.9983>,<at 0.9977 1 RCIM_Cubic>]}
      }}
     CurveIndex=&UniformRangedFloat(Minimum=0, Maximum=360)
    )
    Lifetime=0.2
    Module.Write Color=true
    Module.Write Lifetime=true
    Module.Write Mass=true
    Module.Write Material Random=true
    Module.Write Position=false
    Module.Write Scale=false
    Module.Write SpriteRotation=false
    Module.Write SpriteSize=true
    Position=&SimulationPosition()
    Sprite Size=&Vector2DFromFloat(Value=&UniformRangedFloat(Minimum=100, Maximum=200))
   )]
  particleUpdateScript [
   &ParticleState()
   &SolveForcesAndVelocity(Clamp Velocity=false, Limit Acceleration=false)
   &ScaleColor(
    Scale Alpha=&FloatFromCurve(FloatCurve=DataInterface() {propertyValues {Curve {keys [<at -0.0048 0>,<at 0.1299 0.99>,<at 1.0011 0.002>]}}})
    Scale RGB=<5 5 5>
    ScaleA=true
    ScaleRGB=true
   )
   &ScaleSpriteSize(Scale Factor=&Vector2DFromFloat(Value=&FloatFromCurve(FloatCurve=DataInterface() {propertyValues {Curve {keys [<at 0 0>,<at 0.1222 1>,<at 1.0011 0.002>]}}})))
  ]
  rendererClasses [&NiagaraSpriteRendererProperties]
  renderers [{rendererClass=&NiagaraSpriteRendererProperties}]
 }
 Glow_Partent {
  emitterUpdateScript [
   &EmitterScalability(
    Module.bApplySpawnCountScale=false
    Module.bApplySpawnCountScaleByDistanceFraction=false
    Module.bUseMaxDistance=false
    Module.bUseMinDistance=false
   ) {enabled=false}
   &EmitterLifeCycle(NextLoopDuration=1) {enabled=false}
   &EmitterState(Life Cycle Mode=::NewEnumerator1, Loop Behavior=::NewEnumerator1, Loop Duration=2, UseLoopDelay=false)
   &SpawnBurst_Instantaneous(Spawn Count=15, Spawn Time=0.05, Use Spawn Probability=false) {enabled=false}
   &SpawnRate(SpawnRate=20, Use Spawn Probability=false)
  ]
  particleSpawnScript [
   &InitializeParticle(
    Color=&ColorFromCurve(
     ColorCurve=DataInterface() {propertyValues {
       AlphaCurve {keys [<at -0.0023 1 RCIM_Cubic>]}
       BlueCurve {keys [<at -0.0023 1 RCIM_Cubic>,<at 0.1951 1>,<at 0.4596 0.6202>,<at 0.7292 0.115>,<at 0.9977 0.099 RCIM_Cubic>]}
       GreenCurve {keys [<at -0.0023 0.0677 RCIM_Cubic>,<at 0.1951 0.3883>,<at 0.4596 1>,<at 0.7292 1>,<at 0.9977 0.427 RCIM_Cubic>]}
       RedCurve {keys [<at -0.0023 0.9381 RCIM_Cubic>,<at 0.1951 0.0434>,<at 0.4596 0>,<at 0.7292 0.9983>,<at 0.9977 1 RCIM_Cubic>]}
      }}
     CurveIndex=&UniformRangedFloat(Minimum=0, Maximum=360)
    )
    Lifetime=&UniformRangedFloat(Minimum=0.65, Maximum=1)
    Module.Write Color=true
    Module.Write Lifetime=true
    Module.Write Mass=true
    Module.Write Material Random=true
    Module.Write Position=false
    Module.Write Scale=false
    Module.Write SpriteRotation=false
    Module.Write SpriteSize=true
    Position=&SimulationPosition()
    Sprite Size=&Vector2DFromFloat(Value=&UniformRangedFloat(Minimum=10, Maximum=10))
   )
   &AddVelocity(Velocity=&UniformRangedVector(Minimum=<-750 -450 300>, Maximum=<750 400 700>))
   &ApplyInitialForces()
   &SystemLocation(MaskSpawnGroup=false, Offset=<0 0 75>)
  ]
  particleUpdateScript [
   &ParticleState()
   &Drag(Drag=2.7, Use Linear Drag=true, Use Rotational Drag=false)
   &SolveForcesAndVelocity(Clamp Velocity=false, Limit Acceleration=false)
   &ScaleColor(
    Scale Alpha=&FloatFromCurve(FloatCurve=DataInterface() {propertyValues {Curve {keys [<at -0.0092 0.0144>,<at 0.0967 0.9767>,<at 1.0011 0.002>]}}})
    ScaleA=true
    ScaleRGB=true
   ) {enabled=false}
   &ScaleSpriteSize(Scale Factor=&Vector2DFromFloat(Value=&FloatFromCurve(
      FloatCurve=DataInterface() {propertyValues {Curve {keys [<at 0.02 0.1>,<at 0.4991 0.2393>,<at 0.5756 0.4304>,<at 0.6484 0.2578>,<at 0.7146 0.5166>,<at 0.826 0.2413>,<at 0.9021 0.5351>,<at 0.95 0.002>]}}}
      Scale Curve=25
     )))
   &GenerateDeathEvent()
  ]
  rendererClasses [&NiagaraSpriteRendererProperties]
  renderers [{rendererClass=&NiagaraSpriteRendererProperties}]
 }
 Sparks_Child {
  emitterUpdateScript [
   &EmitterScalability(
    Module.bApplySpawnCountScale=false
    Module.bApplySpawnCountScaleByDistanceFraction=false
    Module.bUseMaxDistance=false
    Module.bUseMinDistance=false
   ) {enabled=false}
   &EmitterLifeCycle(NextLoopDuration=1) {enabled=false}
   &EmitterState(Life Cycle Mode=::NewEnumerator1, Loop Behavior=::NewEnumerator1, Loop Duration=3, UseLoopDelay=false)
   &SpawnBurst_Instantaneous(Spawn Count=15, Spawn Time=0.05, Use Spawn Probability=false) {enabled=false}
  ]
  particleSpawnScript [
   &InitializeParticle(
    Color=&ColorFromCurve(
     ColorCurve=DataInterface() {propertyValues {
       AlphaCurve {keys [<at -0.0023 1 RCIM_Cubic>]}
       BlueCurve {keys [<at -0.0023 1 RCIM_Cubic>,<at 0.1951 1>,<at 0.4596 0.6202>,<at 0.7292 0.115>,<at 0.9977 0.099 RCIM_Cubic>]}
       GreenCurve {keys [<at -0.0023 0.0677 RCIM_Cubic>,<at 0.1951 0.3883>,<at 0.4596 1>,<at 0.7292 1>,<at 0.9977 0.427 RCIM_Cubic>]}
       RedCurve {keys [<at -0.0023 0.9381 RCIM_Cubic>,<at 0.1951 0.0434>,<at 0.4596 0>,<at 0.7292 0.9983>,<at 0.9977 1 RCIM_Cubic>]}
      }}
     CurveIndex=&UniformRangedFloat(Minimum=0, Maximum=360)
    )
    Lifetime=&UniformRangedFloat(Minimum=0.5, Maximum=0.7)
    Module.Write Color=true
    Module.Write Lifetime=true
    Module.Write Mass=true
    Module.Write Material Random=true
    Module.Write Position=false
    Module.Write Scale=false
    Module.Write SpriteRotation=false
    Module.Write SpriteSize=true
    Position=&SimulationPosition()
    Sprite Size=&Vector2DFromFloat(Value=&UniformRangedFloat(Minimum=10, Maximum=30))
   )
   &AddVelocityFromPoint(Module.Use Falloff Distance=false, Velocity Strength=750)
   &ApplyInitialForces()
  ]
  particleUpdateScript [
   &ParticleState()
   &CurlNoiseForce(Noise Frequency=0.02, Noise Strength=300, Pan Noise=false)
   &Drag(Drag=5, Use Linear Drag=true, Use Rotational Drag=false)
   &SolveForcesAndVelocity(Clamp Velocity=false, Limit Acceleration=false)
   &ScaleColor(
    Scale Alpha=&FloatFromCurve(FloatCurve=DataInterface() {propertyValues {Curve {keys [<at -0.0048 0.9952>,<at 1.0011 0.002>]}}})
    Scale RGB=<5 5 5>
    ScaleA=true
    ScaleRGB=true
   )
   &ScaleSpriteSize(Scale Factor=&Vector2DFromFloat(Value=&FloatFromCurve(FloatCurve=DataInterface() {propertyValues {Curve {keys [<at 0 0.9952>,<at 1.0011 0.002>]}}})))
  ]
  rendererClasses [&NiagaraSpriteRendererProperties]
  renderers [{rendererClass=&NiagaraSpriteRendererProperties}]
 }
}
summary {
 systemName=NS_Teddys_vs_Bears_Fireworks
 emitters [
  {emitterName=Glow_Partent, rendererClasses [&NiagaraSpriteRendererProperties]}
  {emitterName=Confetti_Child, rendererClasses [&NiagaraSpriteRendererProperties]}
  {emitterName=Sparks_Child, rendererClasses [&NiagaraSpriteRendererProperties]}
  {emitterName=Confetti_Ribbons_Child, rendererClasses [&NiagaraSpriteRendererProperties]}
  {emitterName=Glow_Child001, rendererClasses [&NiagaraSpriteRendererProperties]}
 ]
 userVariables [
  User.bIsInWater {type=NiagaraBool}
  User.Boost {type=NiagaraFloat, defaultValue=0, description=""}
  User.EffectScale {type=NiagaraFloat, defaultValue=1, description=""}
  User.EffectSpawnScale {type=NiagaraFloat, defaultValue=10, description=""}
  User.ForwardSpeed {type=NiagaraFloat, defaultValue=0, description=""}
  User.FrontPontoonBig.Depth {type=NiagaraFloat, defaultValue=0, description=""}
  User.FrontPontoonBig.Impulse {type=Vector3f, description="", defaultValue {struct=Vector3f, value=<0 0 0>}}
  User.GlobalOffset {type=Vector3f, description="", defaultValue {struct=Vector3f, value=<0 0 0>}}
  User.InWater {type=NiagaraBool}
  User.isBoosting {type=NiagaraBool}
  User.MidPontoonBig.Depth {type=NiagaraFloat, defaultValue=0, description=""}
  User.MidPontoonBig.Impulse {type=Vector3f, description="", defaultValue {struct=Vector3f, value=<0 0 0>}}
  User.NewVariable {type=NiagaraFloat, defaultValue=0, description=""}
  User.RearPontoonBig.Depth {type=NiagaraFloat, defaultValue=0, description=""}
  User.RearPontoonBig.Impulse {type=Vector3f, description="", defaultValue {struct=Vector3f, value=<0 0 0>}}
  User.WaterBodyIndex {type=NiagaraFloat, defaultValue=0, description=""}
  User.WaterDepth {type=NiagaraFloat, defaultValue=0, description=""}
  User.WaterPlaneLocation {type=Vector3f, description="", defaultValue {struct=Vector3f, value=<0 0 0>}}
  User.WaterPlaneNormal {type=Vector3f, description="", defaultValue {struct=Vector3f, value=<0 0 0>}}
  User.WaveTime {type=NiagaraFloat, defaultValue=0, description=""}
 ]
}
